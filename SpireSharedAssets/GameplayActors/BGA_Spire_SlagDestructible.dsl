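; BGA_Spire_SlagDestructible is a destructible Spire slag actor with a weakpoint, marked as superseded by B_Prop_SpireSlag_Shockwave.
; On authority it grants an ability set directly (working around a Scenario spawn bug), starts its StateTree and sets scaled health; remotes disable proximity overlaps.
; The sensor tracks player pawns and sends PlayerDetected to the StateTree for non-teammates.
; Death hides the mesh and destroys the child-actor weakpoint; a custom event plays ricochet VFX.
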(event EventBeginPlay
  ;[0.15 0 0 0.5] USE B_Prop_SpireSlag_Shockwave INSTEAD
  (|CallParentFunction)
  (Utilities|FlowControl|Sequence
    (:then_0
      (Networking|SwitchHasAuthority
        (:Authority
          (Ability|EquipFortAbilitySet self "/SpireSharedAssets/Abilities/AS_Spire_SlagDestructible.AS_Spire_SlagDestructible") ; Inherent Ability Sets bugged when spawning from Scenario
          (AI|Logic|StartLogic (Variables|Default|GetStateTree))
          ;;[0 0.6719 0.1154 0.5] Set Scaled Health Value
          (Health|SetMaxHealthDontCreateASC self (ScalableFloat|GetValueatLevel (Variables|InternalData|GetSF_DefaultHealthLevel) 1.0)))
        (:Remote
          (Variables|/Script/Engine.PrimitiveComponent|SetbGenerateOverlapEvents false (Variables|Default|GetProximitySensor)))))
    (:then_1
      (EventDispatchers|BindEventtoOnDied (delegate Custom|OnDiedEvent) :!class /Script/FortniteGame.BuildingActor)
      (Variables|InternalData|SetSensedPlayerPawnCount 0))))

(event Custom|OnDiedEvent
  (params (DamagedActor :type Actor%20Object%20Reference) (Damage :type Float%20%28single-precision%29) (InstigatedBy :type Controller%20Object%20Reference) (DamageCauser :type Actor%20Object%20Reference) (HitLocation :type Vector) (FHitComponent :type Primitive%20Component%20Object%20Reference) (BoneName :type Name) (Momentum :type Vector))
  ;[0.15 0 0 0.5] USE B_Prop_SpireSlag_Shockwave INSTEAD
  (Rendering|SetVisibility (Variables|Debug|GetStaticMeshComponent) false true)
  (bind _asbga_weakpoint_actor_slag_destructible (Utilities|Casting|CastToBGA_WeakpointActor_SlagDestructible (Variables|/Script/Engine.ChildActorComponent|GetChildActor (Variables|Default|GetSlagWeakpoint))))
  (Class|BGAWeakpointActorBase|UpdateWeakpointCollision _asbga_weakpoint_actor_slag_destructible false "NewEnumerator0" (Variables|/SpireSharedAssets/GameplayActors/BGA_WeakpointActor_SlagDestructible.BGA_WeakpointActor_SlagDestructible_C|GetWeakpointBox _asbga_weakpoint_actor_slag_destructible))
  (Class|BGAWeakpointActorBase|DestroyWeakpoint _asbga_weakpoint_actor_slag_destructible))

(event Custom|SpawnRicochet
  (params (HitResult :type Hit%20Result%20Structure))
  (bind (_bblockinghit _binitialoverlap _time _distance _location _impactpoint _normal) (Collision|BreakHitResult HitResult))
  ;[0.15 0 0 0.5] USE B_Prop_SpireSlag_Shockwave INSTEAD
  ;;[0.9271 0.1015 0.09266 0.5] Play Ricochet Hit VFX
  (bind _returnvalue (Effects|Components|ParticleSystem|SpawnFXAtLocation "/SpireSharedAssets/VFX/NS_SpireBoss_RicochetImpact.NS_SpireBoss_RicochetImpact" _location (Math|Conversions|RotationFromXVector _normal)))
  (Effects|Components|ParticleSystem|SetVectorParameter _returnvalue "ImpactFXDirection" _normal))

(event OnComponentBeginOverlap(ProximitySensor)
  (params (OverlappedComponent :type Primitive%20Component%20Object%20Reference) (OtherActor :type Actor%20Object%20Reference) (OtherComp :type Primitive%20Component%20Object%20Reference) (OtherBodyIndex :type Integer) (bFromSweep :type Boolean) (SweepResult :type Hit%20Result%20Structure%20%28by%20ref%29))
  (bind _sensedplayerpawns (Variables|InternalData|GetSensedPlayerPawns))
  ;[0.15 0 0 0.5] USE B_Prop_SpireSlag_Shockwave INSTEAD
  ;;[0.9479 0.3295 0.05314 0.5] Monitor Player Pawn Sensors
  (bind _asfort_player_pawn (Utilities|Casting|CastToFortPlayerPawn OtherActor))
  (if (not (Team|OnSameTeam _asfort_player_pawn self))
    (Utilities|Array|AddUnique _sensedplayerpawns _asfort_player_pawn)
    (Variables|InternalData|SetSensedPlayerPawnCount (Utilities|Array|Length _sensedplayerpawns))
    (Gameplay|StateTree|SendStateTreeEvent (Variables|Default|GetStateTree) (Utilities|Struct|MakeStateTreeEvent "(TagName=\"StateTree.Event.SlagDestructible.PlayerDetected\")") :!class /Script/GameplayStateTreeModule.StateTreeComponent)))

(event OnComponentEndOverlap(ProximitySensor)
  (params (OverlappedComponent :type Primitive%20Component%20Object%20Reference) (OtherActor :type Actor%20Object%20Reference) (OtherComp :type Primitive%20Component%20Object%20Reference) (OtherBodyIndex :type Integer))
  (bind _sensedplayerpawns (Variables|InternalData|GetSensedPlayerPawns))
  ;[0.15 0 0 0.5] USE B_Prop_SpireSlag_Shockwave INSTEAD
  ;;[0.9479 0.3295 0.05314 0.5] Monitor Player Pawn Sensors
  (bind _asfort_player_pawn (Utilities|Casting|CastToFortPlayerPawn OtherActor))
  (Utilities|Array|RemoveItem _sensedplayerpawns _asfort_player_pawn)
  (Variables|InternalData|SetSensedPlayerPawnCount (Utilities|Array|Length _sensedplayerpawns)))

(fn ConstructionScript)
