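; BGA_WeakpointActor_SlagDestructible is a weakpoint actor Blueprint, named for slag destructibles, that switches its weakpoint on when play begins.
; After the parent BeginPlay runs, it sets the weakpoint box's collision kind, updates that collision as active with the same kind, then enables the weakpoint.
; The construction script only calls the parent, so all other behavior comes from the inherited class.
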(event EventBeginPlay
  (bind _weakpointbox (Variables|Default|GetWeakpointBox))
  (|CallParentFunction)
  (CallFunction|SetWeakpointCollisionKind "NewEnumerator0" _weakpointbox)
  (CallFunction|UpdateWeakpointCollision true "NewEnumerator0" _weakpointbox)
  (CallFunction|EnableWeakpoint))

(fn ConstructionScript
  (|CallParentFunction))
