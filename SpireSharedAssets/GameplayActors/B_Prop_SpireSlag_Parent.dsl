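; Parent Blueprint for Spire slag props, which are destructible building actors with health and a death effect.
; On BeginPlay the authority forces health from a scalable float; server death fires a burst gameplay cue at the hit.
; It destroys itself when its linked actor is destroyed.
; SetOrganicTransform, uncalled here, randomizes scale, yaw and pitch, mapping scale to Z offset so larger pieces sink deeper by default.
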(event EventBeginPlay
  (Utilities|FlowControl|Sequence
    (:then_0
      (Networking|SwitchHasAuthority
        (:Authority
          (Gameplay|ForceBuildingHealth (ScalableFloat|GetValueatLevel (Variables|Default|GetHF_SlagHealth))))))))

(event EventOnDeathServer
  (params (Damage :type Float%20%28single-precision%29) (DamageTags :type Gameplay%20Tag%20Container%20Structure%20%28by%20ref%29) (Momentum :type Vector :default "0, 0, 0") (HitInfo :type Hit%20Result%20Structure%20%28by%20ref%29) (InstigatedBy :type Controller%20Object%20Reference) (DamageCauser :type Actor%20Object%20Reference) (EffectContext :type Gameplay%20Effect%20Context%20Handle%20Structure))
  (GameplayCue|ExecuteGameplayCueOnActor(Burst) self (Variables|Default|GetGC_Death) (GameplayCue|MakeGameplayCueParametersfromHitResult HitInfo)))

(event OnLinkedActorDestroyed(FortLinkToActor)
  (params (DamageCauser :type Actor%20Object%20Reference))
  (Actor|DestroyActor))

(fn SetOrganicTransform
  (params (ScaleRangeMult :type Vector%202D%20Structure :default "(X=0.750000,Y=1.500000)") (RandomYawAdd :type Vector%202D%20Structure :default "(X=0.000000,Y=360.000000)") (RandomPitchAdd :type Vector%202D%20Structure :default "(X=-10.000000,Y=10.000000)") (RandomZAdd :type Vector%202D%20Structure :default "(X=0.000000,Y=-10.000000)"))
  (bind (_randomyaw_x _randomyaw_y) (Math|Vector2D|BreakVector2D (Variables|GetRandomYaw_X)))
  (bind (_originalscale_x _originalscale_y _originalscale_z) (Math|Vector|BreakVector (Variables|GetOriginalScale_X)))
  (bind (_originallocation_x _originallocation_y _originallocation_z) (Math|Vector|BreakVector (Variables|GetOriginalLocation_X)))
  (bind (_originalrotation_roll _originalrotation_pitch _originalrotation_yaw) (Math|Rotator|BreakRotator (Variables|GetOriginalRotation_Roll)))
  (bind (_scalerange_x _scalerange_y) (Math|Vector2D|BreakVector2D (Variables|GetScaleRange_X)))
  (bind (_scalerange_x_2 _scalerange_y_2) (Math|Vector2D|BreakVector2D (Variables|GetScaleRange_X)))
  (bind (_randomz_x _randomz_y) (Math|Vector2D|BreakVector2D (Variables|GetRandomZ_X)))
  (bind (_randompitch_x _randompitch_y) (Math|Vector2D|BreakVector2D (Variables|GetRandomPitch_X)))
  (LocalVariables|SetTarget self)
  (LocalVariables|SetScaleRange ScaleRangeMult)
  (LocalVariables|SetRandomYaw RandomYawAdd)
  (LocalVariables|SetRandomPitch RandomPitchAdd)
  (LocalVariables|SetRandomZ RandomZAdd)
  (bind _output_get (LocalVariables|SetOriginalTransform (Transformation|GetActorTransform (LocalVariables|GetTarget))))
  (bind (_output_get_location _output_get_rotation _output_get_scale) (Math|Transform|BreakTransform _output_get))
  (LocalVariables|SetOriginalLocation _output_get_location)
  (LocalVariables|SetOriginalRotation _output_get_rotation)
  (LocalVariables|SetOriginalScale _output_get_scale)
  (LocalVariables|SetRandomScale (Math|Random|RandomFloatinRange _scalerange_x_2 _scalerange_y_2))
  (Transformation|SetActorTransform (LocalVariables|GetTarget) (Math|Transform|MakeTransform (Math|Vector|MakeVector _originallocation_x _originallocation_y (+ _originallocation_z (Math|Float|MapRangeClamped (LocalVariables|GetRandomScale) _scalerange_x _scalerange_y _randomz_x _randomz_y))) (Math|Rotator|MakeRotator _originalrotation_roll (+ _originalrotation_pitch (Math|Random|RandomFloatinRange _randompitch_x _randompitch_y)) (+ _originalrotation_yaw (Math|Random|RandomFloatinRange _randomyaw_x _randomyaw_y))) (Math|Vector|MakeVector (* _originalscale_x (LocalVariables|GetRandomScale)) (* _originalscale_y (LocalVariables|GetRandomScale)) (* _originalscale_z (LocalVariables|GetRandomScale)))) false false))

(fn ConstructionScript)
