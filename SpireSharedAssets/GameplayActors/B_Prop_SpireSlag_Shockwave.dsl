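; Spire slag destructible prop that fires shockwave abilities when damaged or when enemy players come near.
; On authority it equips its ability set and registers perception, which runs a looping pulse timer until sensed players leave.
; Melee hits, or any damage if enabled, trigger a shockwave unless the instigator is the Mang boss.
; Incoming damage plays a reused impact effect and scales vehicle damage; Spire.Minor owners activate an aura.
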(event EventBeginPlay
  (bind _returnvalue (Actor|GetOwner))
  (|CallParentFunction)
  (Utilities|FlowControl|Sequence
    (:then_0
      (Networking|SwitchHasAuthority
        (:Authority
          (Ability|EquipFortAbilitySet self "/SpireSharedAssets/Abilities/AS_Spire_SlagDestructible.AS_Spire_SlagDestructible") ; Inherent Ability Sets bugged when spawning from Scenario
          (Default|RegisterNearbyActorsPerception))))
    (:then_1
      (bind _returnvalue_2 (GameplayTags|HasMatchingGameplayTag(Message) _returnvalue "(TagName=\"Spire.Minor\")"))
      (if _returnvalue_2
        (Components|Activation|Activate (Variables|Default|GetFX_Aura) false)
        (Niagara|SetNiagaraVariable(Vector3) (Variables|Default|GetFX_Aura) "SystemDirection" (Transformation|GetActorLocation _returnvalue))))))

(event EventOnDamageServer
  (params (Damage :type Float%20%28single-precision%29) (DamageTags :type Gameplay%20Tag%20Container%20Structure%20%28by%20ref%29) (Momentum :type Vector :default "0, 0, 0") (HitInfo :type Hit%20Result%20Structure%20%28by%20ref%29) (InstigatedBy :type Controller%20Object%20Reference) (DamageCauser :type Actor%20Object%20Reference) (EffectContext :type Gameplay%20Effect%20Context%20Handle%20Structure))
  ; Any Damage Charges Shockwave
  (|CallParentFunction Damage DamageTags Momentum HitInfo InstigatedBy DamageCauser EffectContext)
  (bind _returnvalue (GameplayTags|HasMatchingGameplayTag(Message) (Pawn|GetControlledPawn InstigatedBy) "(TagName=\"Pawn.Athena.NPC.Mang.Boss\")"))
  (if (not _returnvalue)
    (if (GameplayTags|HasTag DamageTags "(TagName=\"Gameplay.Action.Attacking.Melee\")" true)
      (Knockback|Damaged|StartTakingDamageKnockback true)
      (elif (ScalableFloat|GetValueasBool (Variables|Default|GetHF_AllowAnyDamage))
        (Knockback|Damaged|StartTakingDamageKnockback false)))))

(event Custom|PlayHitImpact
  (params (HitResult :type Hit%20Result%20Structure))
  (bind (_bblockinghit _binitialoverlap _time _distance _location _impactpoint _normal _impactnormal) (Collision|BreakHitResult HitResult))
  (bind _returnvalue (Math|Conversions|RotationFromXVector _impactnormal))
  ;[0.9271 0.1015 0.09266 0.5] Play Hit VFX
  (bind _impactfx (Variables|Default|GetImpactFX)
    (:then
      (Components|Activation|Activate _impactfx true)
      (Transformation|SetWorldLocationAndRotation _impactfx _impactpoint _returnvalue false false))
    (:else
      (bind _returnvalue_2 (Effects|Components|ParticleSystem|SpawnFXAtLocation "/SpireSharedAssets/FX/NS_SpireShards_Impact.NS_SpireShards_Impact" _impactpoint _returnvalue))
      (Variables|Default|SetImpactFX _returnvalue_2))))

(event Custom|OnNearbyActorDetected
  (params (AllFoundActors :type Array%20of%20Actor%20Object%20References) (EnteringActors :type Array%20of%20Actor%20Object%20References) (ExitingActors :type Array%20of%20Actor%20Object%20References))
  (bind _sensedplayerpawns (Variables|Default|GetSensedPlayerPawns))
  (bind _sensedplayerpawns_2 (Variables|Default|GetSensedPlayerPawns))
  ; Detect Players
  (if (Variables|Default|GetUseProximityActivation)
    (for _array_element EnteringActors
      (if (Default|IsValidTarget _array_element)
        (Networking|SwitchHasAuthority
          (:Authority
            (Utilities|Array|AddUnique _sensedplayerpawns_2 (.AsFort%20Player%20Pawn (Default|IsValidTarget _array_element)))
            (Variables|Default|SetSensedPlayerPawnCount (Utilities|Array|Length _sensedplayerpawns_2))
            (Knockback|Proximity|TryStartProximityKnockback)))))
    (for _array_element_2 ExitingActors
      (if (Default|IsValidTarget _array_element_2)
        (Networking|SwitchHasAuthority
          (:Authority
            (Utilities|Array|RemoveItem _sensedplayerpawns (.AsFort%20Player%20Pawn (Default|IsValidTarget _array_element_2)))
            (Variables|Default|SetSensedPlayerPawnCount (Utilities|Array|Length _sensedplayerpawns))
            (if (Utilities|Array|IsEmpty _sensedplayerpawns)
              (Utilities|Time|ClearandInvalidateTimerbyHandle (Variables|Default|GetLoopPulseHandle)))))))))

(event Custom|LoopPulse
  ; Loop Shockwave on perception
  (bind _returnvalue (Abilities|TryActivateAbilitybyClass (Ability|GetFortAbilitySystemComponent self) "/SpireSharedAssets/Abilities/SlagShockwave/GA_Slag_Shockwave.GA_Slag_Shockwave_C" true))
  (if _returnvalue
    (Variables|Default|SetPickaxeActivated false)
    (bind _returnvalue_2 (Utilities|Time|SetTimerbyFunctionName 0 "LoopPulse" (ScalableFloat|GetValueatLevel (Variables|Default|GetPulseLoopDelay)) true))
    (Variables|Default|SetLoopPulseHandle _returnvalue_2)))

(fn BlueprintModifyIncomingDamage
  (params (Damage :type Float%20%28single-precision%29) (DamageTags :type Gameplay%20Tag%20Container%20Structure%20%28by%20ref%29) (EffectContext :type Gameplay%20Effect%20Context%20Handle%20Structure%20%28by%20ref%29) (EventInstigator :type Controller%20Object%20Reference) (DamageCauser :type Actor%20Object%20Reference))
  (bind _returnvalue (|CallParentFunction Damage DamageTags EffectContext EventInstigator DamageCauser))
  (CallFunction|PlayHitImpact (Ability|EffectContext|GetHitResult EffectContext))
  (Utilities|Casting|CastToFortAthenaVehicle DamageCauser
    (:then
      (return (* _returnvalue (ScalableFloat|GetValueatLevel (Variables|Default|GetVehicleDamagePercentage)))))
    (:CastFailed
      (return _returnvalue))))

(fn DoesTakingDamageRestartKnockbackTimer
  (return true))

(fn isValidTarget
  (params (Object :type Object%20Reference))
  (bind _asfort_player_pawn (Utilities|Casting|CastToFortPlayerPawn Object))
  (return (and (not (GameplayTags|HasMatchingGameplayTag _asfort_player_pawn "(TagName=\"Pawn.Athena.NPC\")")) (not (Team|OnSameTeam _asfort_player_pawn self))) _asfort_player_pawn))

(fn RegisterNearbyActorsPerception
  (Networking|SwitchHasAuthority
    (:Authority
      (CallFunction|UnregisterNearbyActorsPerception)
      (if (Variables|Default|GetUseProximityActivation)
        (bind _returnvalue (Query|BPRegisterListener (Variables|Default|GetFortNearbyActorsPerception) (Utilities|Struct|MakeFortNearbyActorsPerceptionConfiguration 1.0 3 0 (ScalableFloat|GetValueatLevelwithContext (Variables|Default|GetPerceptionRadius))) (EventDispatchers|CreateEvent :!event onNearbyActorDetected)))
        (Variables|Default|SetPerceptionListenerHandle _returnvalue)))))

(fn StartProximityKnockback
  (bind _returnvalue (Utilities|Time|SetTimerbyFunctionName 0 "LoopPulse" (ScalableFloat|GetValueatLevel (Variables|Default|GetInitPulseLoopDelay))))
  (Variables|Default|SetLoopPulseHandle _returnvalue))

(fn StartTakingDamageKnockback
  (params (Pickaxe :type Boolean))
  (if (Knockback|Damaged|DoesTakingDamageRestartKnockbackTimer)
    (bind _returnvalue (Abilities|TryActivateAbilitybyClass (Ability|GetFortAbilitySystemComponent self) (Variables|Default|GetShockwaveGA) true))
    (if _returnvalue
      (Variables|Default|SetPickaxeActivated (Variables|GetPickaxe)))))

(fn TryStartProximityKnockback
  (if (not (Utilities|Time|IsTimerActivebyHandle (Variables|Default|GetLoopPulseHandle)))
    (Knockback|Proximity|StartProximityKnockback)))

(fn UnregisterNearbyActorsPerception
  (Networking|SwitchHasAuthority
    (:Authority
      (Query|BPUnregisterListener (Variables|Default|GetPerceptionListenerHandle))
      (Utilities|Array|Clear (Variables|Default|GetSensedPlayerPawns))
      (Variables|Default|SetSensedPlayerPawnCount 0))))

(fn ConstructionScript
  (|CallParentFunction))
