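; Spire spike prop actor that grows to full size and emits a shockwave that applies a knockback effect to nearby pawns.
; StartSpikeGrowth fires a telegraph cue, waits a hotfixable delay, then asks its scaler component to reach max scale.
; On success it fires a burst cue and sphere-traces around itself to apply the knockback GameplayEffect.
; The server records initial scale in a variable whose OnRep applies it; the bound scaling-finished handler is empty.
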(event EventBeginPlay
  (|CallParentFunction)
  (Networking|SwitchHasAuthority
    (:Authority
      (Default|BindEventtoSpikeScalingFinished (Variables|Default|GetComponent_SpireSpike_SclarComponent) (EventDispatchers|CreateEvent :!event Custom|ScaleFinished))
      (Variables|Internal|SetSpireSpikeScale (Transformation|GetActorScale3D self)))))

(event Custom|StartSpikeGrowth
  (CallFunction|ScaleTelegraph)
  (Utilities|FlowControl|Delay (ScalableFloat|GetValueatLevel (Variables|Hotfix|GetScaleShardDelay_Hf)))
  (bind _scalesuccessful (Class|ComponentSpireSpikeSclarComponent|ScaleSpiketoMax (Variables|Default|GetComponent_SpireSpike_SclarComponent) true))
  (if _scalesuccessful
    (GameplayCue|ExecuteGameplayCueOnActor(Burst) self (Variables|GameplayCue|GetGC_SpireShardBurst) (Ability|GameplayCue|MakeGameplayCueParameters 0.0))
    (bind _outhits (Collision|MultiSphereTraceByChannel :Start (Transformation|GetActorLocation) :End (+ (Transformation|GetActorLocation) "0.000000,0.000000,1.000000") :Radius 2.5e+02 :TraceChannel "TraceTypeQuery9" :bTraceComplex false :DrawDebugType "None" :bIgnoreSelf true :TraceColor "(R=1.000000,G=0.000000,B=0.000000,A=1.000000)" :TraceHitColor "(R=0.000000,G=1.000000,B=0.000000,A=1.000000)" :DrawTime 2.0))
    (for _array_element _outhits
      (bind _asfort_pawn (Utilities|Casting|CastToFortPawn (.HitActor (Collision|BreakHitResult _array_element))))
      (Fortnite|ApplyGameplayEffectsToActor _asfort_pawn (Utilities|Array|MakeArray "/SpireSharedAssets/Abilities/SlagShockwave/GE_GrowingSlag_Knockback.GE_GrowingSlag_Knockback_C") (Transformation|GetActorLocation) false 0 (Ability|GetAbilitySystemComponent 0)))))

(event Custom|ScaleTelegraph
  (GameplayCue|ExecuteGameplayCueOnActor(Burst) self (Variables|GameplayCue|GetGC_SpireShardTelegraph) (Ability|GameplayCue|MakeGameplayCueParameters 0.0)))

(event Custom|ScaleFinished)

(fn GetDamageInstigator
  (return 0))

(fn OnRep_SpireSpikeScale
  (Transformation|SetActorScale3D (Variables|Internal|GetSpireSpikeScale)))

(fn ConstructionScript
  (|CallParentFunction))
