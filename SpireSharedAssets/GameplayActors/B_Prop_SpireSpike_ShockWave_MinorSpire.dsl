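; Spire spike prop actor, named for the minor spire shockwave, that plays a spawn effect when it enters play.
; BeginPlay runs the parent logic, calls SetOrganicTransform with a two-value vector, then calls SpawnFX.
; SpawnFX plays an enlarged, auto-destroying henchman spawn-portal Niagara system at the actor's location, without pooling.
; The construction script only calls the parent.
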(event EventBeginPlay
  (|CallParentFunction)
  (CallFunction|SetOrganicTransform "(X=0.750000,Y=1.500000)")
  (CallFunction|SpawnFX))

(event Custom|SpawnFX
  (Niagara|SpawnSystematLocation :SystemTemplate "/SpireSharedAssets/Assets/VFX/NS_Henchman_SpawnPortal.NS_Henchman_SpawnPortal" :Location (Transformation|GetActorLocation) :Scale "3.000000,3.000000,3.000000" :bAutoDestroy true :bAutoActivate true :PoolingMethod "None" :bPreCullCheck true))

(fn ConstructionScript
  (|CallParentFunction))
